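# Conversation database for the LycheeNickel supply specialist NPC, voiced as Scooby with flavor greetings.
# The entry point leads to one speech node, whose default line can be replaced by contextual messages for a first-ever conversation, per-match conversation count or specific player outfits.
# Priority curve rows rank first-interaction and outfit reactions, while count-based lines have none assigned.
# The speech node then links out to the shared NPC library Quests and BasicServices conversations.
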ConversationDatabase /NPCLibrary/NPCs/LycheeNickel/NPCLibrary_Conversation_LycheeNickel
entry_points [{tag=Conversation.Entry.LycheeNickel, destinations [#1]}]
nodes [
 ConversationEntryPointNode_7 {
  id=#1
  class=ConversationEntryPointNode
  outputs [#2]
  properties {entryTag {tagName=Conversation.Entry.LycheeNickel}}
 }
 Speech_C_4 {
  id=#2
  class=Speech_C
  outputs [#3, #4]
  properties {
   nodeName=""
   generalConfig {
    defaultMessage=(Scooby appears to be chewing on a Scooby Snack.)
    contextualMessages [
     {
      message="Rooby rooby roooo!"
      contextRequirements [{requirement {class=FortControllerRequirement_IsFirstEverConversationWithNPC}}]
      priority {curve {rowName=Tandem.Speech.Priority.FirstEverInteraction}}
     }
     {
      message=Ronsters? Rhere??
      contextRequirements [{requirement {
         class=FortControllerRequirement_CheckNPCMatchConversationCount
         properties {nPCUniqueIdTag {tagName=AISpawnerData.Type.Specialist.Supply.LycheeNickel}, requiredConversationCountConfig {requirement=GreaterThan}}
        }}]
      priority {curve {curveTable=null, rowName=None}}
     }
     {
      message=Ruh roh...
      contextRequirements [{requirement {
         class=FortControllerRequirement_CheckNPCMatchConversationCount
         properties {nPCUniqueIdTag {tagName=AISpawnerData.Type.Specialist.Supply.LycheeNickel}, requiredConversationCountConfig {requirement=GreaterThan}}
        }}]
      priority {curve {curveTable=null, rowName=None}}
     }
     {
      message=(Scooby appears to be chewing on a Scooby Snack.)
      contextRequirements [{requirement {
         class=FortControllerRequirement_CheckNPCMatchConversationCount
         properties {nPCUniqueIdTag {tagName=AISpawnerData.Type.Specialist.Supply.LycheeNickel}, requiredConversationCountConfig {requirement=GreaterThan}}
        }}]
      priority {curve {curveTable=null, rowName=None}}
     }
     {
      message="Rello Raggy!"
      contextRequirements [{requirement {class=FortControllerRequirement_HasCID, properties {allowedCIds [/BRCosmetics/Athena/Items/Cosmetics/Characters/Character_FluteLamp]}}}]
      priority {curve {rowName=Tandem.Speech.Priority.OutfitReaction}}
     }
     {
      message=(Scooby wags his tail happily.)
      contextRequirements [{requirement {
         class=FortControllerRequirement_HasCID
         properties {allowedCIds [
           /BRCosmetics/Athena/Items/Cosmetics/Characters/Character_ArcanaAgate
           /BRCosmetics/Athena/Items/Cosmetics/Characters/Character_OboeThorn
           /BRCosmetics/Athena/Items/Cosmetics/Characters/Character_BaskIsle
          ]}
        }}]
      priority {curve {rowName=Tandem.Speech.Priority.OutfitReaction}}
     }
    ]
   }
  }
 }
 ConversationLinkNode_0 {id=#4, class=ConversationLinkNode, properties {remoteEntryTag {tagName=Conversation.Entry.BasicServices}}}
 ConversationLinkNode_1 {id=#3, class=ConversationLinkNode, properties {remoteEntryTag {tagName=Conversation.Entry.NPCLibrary.Quests}}}
]
